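# Translucent UI material, apparently for item shop offer tiles, drawing animated bubbles over colored gradient backgrounds.
# Counter-rotating noise samples distort a bubble texture, whose stepped mask is faded by a vertical gradient and blended into the bubbles pass.
# Particle grids build randomized, drifting bubble cells with per-cell scale and random culling.
# Named reroutes share background parameter colors, inner glow, banner-dependent image alpha and a sine-wave progress mask driven by OverlayProgress.
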Material /OfferCatalog/Art/Safflower/M_UI_Bubbles
graph {
 properties {
  bAutomaticallySetUsageInEditor=false
  bUsedWithStaticMesh=false
  blendMode=BLEND_Translucent
  materialDomain=MD_UI
  usedShadingModels=Default Lit
 }
 nodes [
  Constant() {r=1}
  LinearInterpolate(A=NamedRerouteUsage(), B=NamedRerouteUsage(), Alpha=NamedRerouteUsage())
  MaterialFunctionCall(1=NamedRerouteUsage(), 2=NamedRerouteUsage(), 3=NamedRerouteUsage(), A=NamedRerouteUsage()) {materialFunction=&Lerp_Multiple_Float3}
  MaterialFunctionCall() {materialFunction=&GetUserInterfaceUV}
  MaterialFunctionCall!(Input Alpha=Constant() {r=1}, bOverrideCornerRadius, Corner Radius Override) {materialFunction=/FortUILibrary/Systems/ItemShopTile/Materials/MaterialFunctions/MF_UI_ItemShop_RoundedCorners}
  n1 = NamedRerouteUsage()
  n2 = ScalarParameter() {defaultValue=0.5, group=Bubbles, parameterName=Bubbles_Scale, sliderMax=2, sliderMin=0.1}
  n3 = MaterialFunctionCall!(ScreenUV, TextureSize, ScalePivotPoint) {materialFunction=/FortUILibrary/MaterialFunctions/Basics/MF_UI_ScreenSpaceUVScaling}
  n4 = Multiply(A=Add(A=Time(), B=NamedRerouteUsage()), B=ScalarParameter() {defaultValue=0.065, group=Bubbles, parameterName=Bubbles_SpeedMultiply}) {constB=0.1}
  n5 = NamedRerouteUsage()
  n6 = Multiply(A=@n4) {constB=-0.05}
  n7 = DotProduct(
   A=TextureSample(UVs=Rotator(Coordinate=@n3.ScaleY, Time=@n4) {speed=6.2832}, Tex=@n5) {texture=&T_Smooth_tile}.R
   B=TextureSample(UVs=Rotator(Coordinate=@n3.ScaleY, Time=@n6) {speed=6.2832}, Tex=@n5) {texture=&T_Smooth_tile}.R
  )
  n8 = ScalarParameter() {defaultValue=0.8, group=Bubbles, parameterName=Bubbles_Distortion, sliderMax=0.5}
  n9 = TextureSample(UVs=Add(
    A=MaterialFunctionCall(UVs=@n1, Texture Scale=@n2) {materialFunction=&ScaleUVsByCenter}.UVs
    B=Multiply(A=@n7, B=@n8) {constB=0.1, desc=Intensity}
   )) {samplerSource=SSM_Clamp_WorldGroupSettings, samplerType=SAMPLERTYPE_LinearColor, texture=/OfferCatalog/Art/Ch6MS2/T_UI_Ch6MS2_Bubbles}
  n10 = MaterialFunctionCall!("Value (0-1)"=NamedRerouteUsage(), Steps=Constant() {r=3}) {materialFunction=/FortUILibrary/MaterialFunctions/Basics/MF_UI_StepRatio}
  n11 = Multiply(
   A=NamedRerouteDeclaration(None=MaterialFunctionCall!(1=@n9.R, 2, 3=@n9.R, 4, A=@n10.Stepped Gradient) {materialFunction=/Engine/Functions/Engine_MaterialFunctions02/Lerp_Multiple_Float}.Lerp 3 Inputs) {name=Bubbles Mask, nodeColor=<rgba 0 0 0 1>}
   B=LinearInterpolate(Alpha=NamedRerouteUsage())
  )
  n12 = ComponentMask(None=NamedRerouteUsage()) {g=true}
  n13 = NamedRerouteDeclaration(None=Power(Base=OneMinus(None=@n12)) {constExponent=8}) {name=Bubbles Vertical Gradient Mask, nodeColor=<rgba 0 0 0 1>}
  Multiply(A=@n11, B=OneMinus(None=@n13))
  NamedRerouteDeclaration(None=MaterialFunctionCall(UV=NamedRerouteUsage()) {materialFunction=&MF_UI_RadialGradient}.Smooth) {name=Radial Gradient, nodeColor=<rgba 0 0 0 1>}
  n14 = ScalarParameter() {parameterName=bIsBanner, sliderMax=1}
  NamedRerouteDeclaration(None=LinearInterpolate(A=NamedRerouteUsage(), B=NamedRerouteUsage(), Alpha=NamedRerouteDeclaration(None=@n14) {name=Is Banner, nodeColor=<rgba 0 0 0 1>})) {name=Image Alpha, nodeColor=<rgba 0 0 0 1>}
  NamedRerouteDeclaration(None=NamedRerouteUsage()) {name=Image Color, nodeColor=<rgba 0 0 0 1>}
  NamedRerouteDeclaration(None=MaterialFunctionCall() {materialFunction=&GetUserInterfaceUV}.9-Slice UV) {name=9 Slice UV, nodeColor=<rgba 0 0 0 1>}
  n15 = Add(
   A=NamedRerouteUsage()
   B=LinearInterpolate(Alpha=ScalarParameter() {defaultValue=1, group=Inner Glow, parameterName=Inner Glow Size, sliderMax=1}) {constB=0.5}
  ) {constB=0.5}
  n16 = Saturate(None=Multiply(A=Saturate(None=@n15)) {constB=2})
  NamedRerouteDeclaration(None=Multiply(
    A=NamedRerouteDeclaration(None=@n16) {name=Inner Glow Mask, nodeColor=<rgba 0 0 0 1>}
    B=ScalarParameter() {defaultValue=0.3, group=Inner Glow, parameterName=Inner Glow Opacity, sliderMax=1}
   )) {name=Inner Glow Opacity, nodeColor=<rgba 0 0 0 1>}
  NamedRerouteDeclaration(None=NamedRerouteUsage()) {name=Combined Background-Icon, nodeColor=<rgba 0 0 0 1>}
  n17 = Divide(A=NamedRerouteUsage(), B=LinearInterpolate(Alpha=@n13) {constA=0.4}) {constB=0.4}
  NamedRerouteDeclaration(None=LinearInterpolate(A=NamedRerouteUsage(), B=Saturate(None=@n17), Alpha=@n11)) {name=Bubbles Pass, nodeColor=<rgba 0 0 0 1>}
  NamedRerouteDeclaration(None=TextureObject() {samplerType=SAMPLERTYPE_Grayscale, texture=/FortWater/Textures/LowResBlurredNoise}) {name=Noise Texture, nodeColor=<rgba 0 0 0 1>}
  NamedRerouteDeclaration(None=ScalarParameter() {group=Edge Noise, parameterName=RandomSeed, sliderMax=1}) {name=Random Seed, nodeColor=<rgba 0 0 0 1>}
  n18 = NamedRerouteUsage()
  n19 = NamedRerouteUsage()
  NamedRerouteDeclaration(None=LinearInterpolate(
    A=NamedRerouteDeclaration(None=@n18) {name=BG Color, nodeColor=<rgba 0 0 0 1>}
    B=NamedRerouteDeclaration(None=@n19) {name=Glow Color, nodeColor=<rgba 0 0 0 1>}
    Alpha=NamedRerouteUsage()
   )) {name=Background Colors, nodeColor=<rgba 0 0 0 1>}
  n20 = NamedRerouteUsage()
  n21 = NamedRerouteUsage()
  n22 = MaterialFunctionCall(
   UVs=MaterialFunctionCall() {materialFunction=&MF_UI_UVScaling}.ScaleY
   GradientStart=Constant() {r=0.4}
   GradientEnd=Constant() {r=1}
   RotationAngle=Constant() {r=0.25}
  ) {materialFunction=&MF_UI_LinearGradient}
  n23 = Multiply(A=@n22.Result) {constB=1.1}
  n24 = Multiply(A=NamedRerouteUsage())
  n25 = Constant() {r=0.1}
  NamedRerouteDeclaration(None=Add(A=LinearInterpolate(A=@n20, B=@n21, Alpha=@n23), B=Multiply(A=@n24, B=@n25))) {name=Bubbles Background, nodeColor=<rgba 0 0 0 1>}
  NamedRerouteDeclaration(None=TextureSample(Tex=NamedRerouteUsage()) {texture=&T_Smooth_tile}.R) {name=Water Noise, nodeColor=<rgba 0 0 0 1>}
  NamedRerouteDeclaration(None=ScalarParameter() {defaultValue=1, group=Shape, parameterName=OverlayProgress, sliderMax=1, sortPriority=2}) {name=Progression Value, nodeColor=<rgba 0 0 0 1>}
  NamedRerouteDeclaration(None=MaterialFunctionCall() {materialFunction=&MF_UI_UVScaling}.ScaleY) {name=Normalized UV, nodeColor=<rgba 0 0 0 1>}
  n26 = ScalarParameter() {defaultValue=0.1, group=0 - Global, parameterName=Global Speed}
  NamedRerouteDeclaration(None=Multiply(A=Time(), B=@n26)) {name=Global Speed, nodeColor=<rgba 0 0 0 1>}
  NamedRerouteDeclaration(None=@n26) {name=Global Speed Input, nodeColor=<rgba 0 0 0 1>}
  n27 = MaterialFunctionCall() {materialFunction=/FortUILibrary/Systems/ItemShopTile/Materials/MaterialFunctions/MF_UI_ItemShop_BackgroundParams}
  NamedRerouteDeclaration(None=@n27.Color 1) {name=color 1, nodeColor=<rgba 0.2422 0 1 1>}
  n28 = NamedRerouteUsage()
  n29 = ScalarParameter() {defaultValue=1, group=Bubbles, parameterName=RandomColumnsAmount}
  n30 = Multiply(A=Time()) {constB=0.002}
  n31 = AppendVector(A=Constant(), B=Add(A=Multiply(A=@n28, B=@n29) {constB=5}, B=Frac(None=@n30)))
  n32 = ScalarParameter() {group=Bubbles, parameterName=BubbleOffsetRange}
  n33 = Multiply(A=@n32) {constB=-1}
  n34 = NamedRerouteUsage()
  n35 = LinearInterpolate(
   A=ScalarParameter() {defaultValue=0.5, group=03 - Cell, parameterName=Minimum Cell Scale}
   B=ScalarParameter() {defaultValue=1.2, group=03 - Cell, parameterName=Maximum Cell Scale}
   Alpha=@n36.Random Cell Values
  ) {constA=0.5}
  n36 = MaterialFunctionCall!(
   UV=NamedRerouteUsage()
   Global Offset=@n31
   Global Rotation
   Tiling=NamedRerouteUsage()
   Cell Offset=AppendVector(A=LinearInterpolate(A=@n32, B=@n33, Alpha=@n34) {constA=0.05, constB=-0.05}, B=Constant())
   Cell Scale=@n35
   Cell Rotation=Constant()
   Random Seed
  ) {materialFunction=/FortUILibrary/MaterialFunctions/Basics/MF_UI_ParticleGrid}
  n37 = Constant()
  n38 = MaterialFunctionCall(
   UV=NamedRerouteUsage()
   Global Offset=@n31
   Tiling=NamedRerouteUsage()
   Cell Offset=@n37
   Cell Scale=@n35
   Cell Rotation=@n37
   Random Seed=ScalarParameter() {defaultValue=0.5, group=Bubbles, parameterName=RandomSeed_Bubble}
  ) {materialFunction=&MF_UI_ParticleGrid}
  n39 = ScalarParameter() {defaultValue=0.6, parameterName=Bubble Random Mask}
  NamedRerouteDeclaration(None=Multiply(A=@n36.Cells UV, B=Step(Y=@n38.Random Cell Values, X=@n39) {constX=0.6}) {constA=1}) {name=Bubble, nodeColor=<rgba 0 0 0 1>}
  NamedRerouteDeclaration(None=ScalarParameter() {defaultValue=3, group=Bubbles, parameterName=Bubble Amount}) {name=Bubble Amount_1, nodeColor=<rgba 0 0 0 1>}
  NamedRerouteDeclaration(None=Multiply(A=MaterialFunctionCall() {materialFunction=&MF_UI_ScreenSpaceUVScaling}.ScaleY, B=Constant() {r=0.25})) {name=Bubble UV, nodeColor=<rgba 0 0 0 1>}
  NamedRerouteDeclaration(None=@n38.Random Cell Values) {name=Random Bubble, nodeColor=<rgba 0 0 0 1>}
  NamedRerouteDeclaration(None=@n38.Random Rows) {name=RandomRows, nodeColor=<rgba 0 1 0.6719 1>}
  NamedRerouteDeclaration(None=@n27.Color 2) {name=color 2, nodeColor=<rgba 0.0781 1 0 1>}
  NamedRerouteDeclaration(None=@n36.Random Cell Values) {name=RandomBubbleNew, nodeColor=<rgba 0 1 0.0859 1>}
  NamedRerouteDeclaration(None=@n36.Cells UV) {name=BubblesUV_New, nodeColor=<rgba 0 0.5781 1 1>}
  NamedRerouteDeclaration(None=@n36.Random Columns) {name=RandomColumns_Bubble, nodeColor=<rgba 0.1953 1 0 1>}
  NamedRerouteDeclaration(None=@n27.Color 3) {name=color 3, nodeColor=<rgba 1 0 0.3984 1>}
  n40 = NamedRerouteUsage()
  n41 = OneMinus(None=@n40)
  n42 = NamedRerouteUsage()
  n43 = NamedRerouteUsage()
  n44 = Constant() {r=0.05}
  n45 = Add(
   A=MaterialFunctionCall(
    UVs=MaterialFunctionCall(UV=@n42) {materialFunction=&MF_UI_UVScaling}.ScaleX
    "Rotation Angle (0-1)"=Constant() {r=0.25}
   ) {materialFunction=&CustomRotator}.Rotated Values
   B=AppendVector(A=Constant(), B=Multiply(A=@n43, B=@n44) {constB=0.02, desc=Progression Wave Mask Intensity})
  )
  n46 = AppendVector(A=NamedRerouteUsage(), B=OneMinus(None=NamedRerouteUsage()))
  n47 = NamedRerouteDeclaration(None=MaterialFunctionCall!(
    UV=Add(A=@n45, B=@n46)
    Wave Height=Constant() {r=0.05}
    Wave Elevation
    Wave Length=Constant() {r=1.2}
    Wave Speed=Constant() {r=-0.2}
    GlowMin
    GlowMax=Constant() {r=0.025}
    Outline Thickness
    Outline Alignment
    Custom Curve
   ) {materialFunction=/FortUILibrary/MaterialFunctions/Shapes/MF_UI_Wave}.SDF) {name=SineWaveEdge, nodeColor=<rgba 0 0 0 1>}
  n48 = LinearInterpolate(
   A=LinearInterpolate(
    A=@n41
    B=MaterialFunctionCall(DistanceField=@n47) {materialFunction=&MF_UI_AntiAliasing}.Outer
    Alpha=Ceil(None=@n40)
   )
   B=@n41
   Alpha=Floor(None=NamedRerouteUsage())
  )
  NamedRerouteDeclaration(None=OneMinus(None=NamedRerouteDeclaration(None=@n48) {name=Progress Mask, nodeColor=<rgba 0 0 0 1>})) {name=Inverted Progress Mask, nodeColor=<rgba 0 0 0 1>}
  NamedRerouteUsage()
  SmoothStep(Value=@n22.Result) {constMax=0.75}
  VectorParameter() {group=TileBkgd, parameterName=Tile Bkgd Color, defaultValue=<rgba 0.0782 0.4735 0 1>}
  VectorParameter() {group=TileBkgd, parameterName=Tile Glow Color, defaultValue=<rgba 0.4287 0.9301 0.011 1>}
 ]
}
